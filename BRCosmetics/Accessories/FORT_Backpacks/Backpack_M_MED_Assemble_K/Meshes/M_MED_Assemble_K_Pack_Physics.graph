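# Physics asset for the M_MED_Assemble_K backpack cosmetic, giving its front, left, middle and right tubes bodies for secondary motion.
# Each tube is a chain of capsules constrained back to a small sphere at its base bone, with an unconstrained sphere on the tube offset bone.
# Twist stays tight throughout, while swing limits widen toward the tips of the left and right tubes and stay uniform on the middle one.
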PhysicsAsset /BRCosmetics/Accessories/FORT_Backpacks/Backpack_M_MED_Assemble_K/Meshes/M_MED_Assemble_K_Pack_Physics
bodies {
 dyn_front_tube_base [{shapeType=Sphere, radius=0.5, shapeName=dyn_front_tube_base_sphere}]
 dyn_front_tube_parent [{shapeType=Capsule, length=3.8963, radius=1.3771, shapeName=dyn_front_tube_parent_capsule, center=<0 0 1.1646>, rotation=<rot 0 0 90>}]
 dyn_l_tube_01 [{shapeType=Capsule, length=3.459, radius=1.7663, shapeName=dyn_l_tube_01_capsule, center=<3.2364 0 0>, rotation=<rot 90 0 0>}]
 dyn_l_tube_02 [{shapeType=Capsule, length=2.7275, radius=1.7456, shapeName=dyn_l_tube_02_capsule, center=<2.9317 0 0>, rotation=<rot 90 0 0>}]
 dyn_l_tube_base [{shapeType=Sphere, radius=0.5, shapeName=dyn_l_tube_base_sphere}]
 dyn_l_tube_parent [{shapeType=Capsule, length=3.1428, radius=1.8256, shapeName=dyn_l_tube_parent_capsule, center=<3.0427 0 0>, rotation=<rot 90 0 0>}]
 dyn_mid_tube_01 [{shapeType=Capsule, length=2.9453, radius=1.5032, shapeName=dyn_mid_tube_01_capsule, center=<2.2839 0 -0.0001>, rotation=<rot 90 0 0>}]
 dyn_mid_tube_02 [{shapeType=Capsule, length=3.0262, radius=1.4956, shapeName=dyn_mid_tube_02_capsule, center=<2.2643 0 -0.0001>, rotation=<rot 90 0 0>}]
 dyn_mid_tube_base [{shapeType=Sphere, radius=0.5, shapeName=dyn_mid_tube_base_sphere}]
 dyn_mid_tube_parent [{shapeType=Capsule, length=2.7178, radius=1.4608, shapeName=dyn_mid_tube_parent_capsule, center=<2.4388 0 -0.0002>, rotation=<rot 90 0 0>}]
 dyn_r_tube_01 [{shapeType=Capsule, length=3.0231, radius=1.4624, shapeName=dyn_r_tube_01_capsule, center=<2.4702 0 0>, rotation=<rot 90 0 0>}]
 dyn_r_tube_02 [{shapeType=Capsule, length=2.7558, radius=1.473, shapeName=dyn_r_tube_02_capsule, center=<2.4722 0 0.0001>, rotation=<rot 90 0 0>}]
 dyn_r_tube_03 [{shapeType=Capsule, length=2.5315, radius=1.3942, shapeName=dyn_r_tube_03_capsule, center=<2.2196 0 0>, rotation=<rot 90 0 0>}]
 dyn_r_tube_base [{shapeType=Sphere, radius=0.5, shapeName=dyn_r_tube_base_sphere}]
 dyn_r_tube_parent [{shapeType=Capsule, length=3.0162, radius=1.4335, shapeName=dyn_r_tube_parent_capsule, center=<2.5817 0 0>, rotation=<rot 90 0 0>}]
 dyn_tube_offset [{shapeType=Sphere, radius=0.5, shapeName=dyn_tube_offset_sphere}]
}
constraints [
 {bone1Name=dyn_mid_tube_parent, bone2Name=dyn_mid_tube_base, swing1LimitDegrees=25, swing2LimitDegrees=25, twistLimitDegrees=3}
 {bone1Name=dyn_mid_tube_01, bone2Name=dyn_mid_tube_parent, swing1LimitDegrees=25, swing2LimitDegrees=25, twistLimitDegrees=3}
 {bone1Name=dyn_mid_tube_02, bone2Name=dyn_mid_tube_01, swing1LimitDegrees=25, swing2LimitDegrees=25, twistLimitDegrees=3}
 {bone1Name=dyn_l_tube_parent, bone2Name=dyn_l_tube_base, swing1LimitDegrees=10, swing2LimitDegrees=10, twistLimitDegrees=3}
 {bone1Name=dyn_l_tube_01, bone2Name=dyn_l_tube_parent, swing1LimitDegrees=20, swing2LimitDegrees=20, twistLimitDegrees=3}
 {bone1Name=dyn_l_tube_02, bone2Name=dyn_l_tube_01, swing1LimitDegrees=30, swing2LimitDegrees=30, twistLimitDegrees=3}
 {bone1Name=dyn_r_tube_parent, bone2Name=dyn_r_tube_base, swing1LimitDegrees=5, swing2LimitDegrees=5, twistLimitDegrees=3}
 {bone1Name=dyn_r_tube_01, bone2Name=dyn_r_tube_parent, swing1LimitDegrees=10, swing2LimitDegrees=10, twistLimitDegrees=3}
 {bone1Name=dyn_r_tube_02, bone2Name=dyn_r_tube_01, swing1LimitDegrees=20, swing2LimitDegrees=20, twistLimitDegrees=5}
 {bone1Name=dyn_r_tube_03, bone2Name=dyn_r_tube_02, swing1LimitDegrees=25, swing2LimitDegrees=25, twistLimitDegrees=5}
 {bone1Name=dyn_front_tube_parent, bone2Name=dyn_front_tube_base, swing1LimitDegrees=25, swing2LimitDegrees=25, twistLimitDegrees=3}
]
