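# Animation Blueprint for the Assemble_K backpack that spins its fan bone and makes the pack react to the wearer's movement.
# Time-driven rotators, chosen by a 2D pawn speed threshold, spin the fan additively.
# The owner's spine bend relative to ref pose, read as yaw, offsets and rolls the root.
# A RigidBody node adds secondary physics in root-bone space, with reduced alpha during skydiving states.
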AnimBlueprint /BRCosmetics/Accessories/FORT_Backpacks/Backpack_M_MED_Assemble_K/Meshes/M_MED_Assemble_K_Pack_AnimBP
target_skeleton=&/M_MED_Assemble_K_Pack_Skeleton
graphs [
 AnimGraph {nodes [
   n1 = "Transform (Modify) Bone - Bone: dyn_fan_parent"(
    ComponentPose=Local To Component(LocalPose=Input Pose())
    Alpha=1
    Translation="0,0,0"
    Rotation=Get Spinning_Rotator() -> (Spinning_Rotator:Rotator)
    Scale="1,1,1"
   ) {
    rotationMode=BMM_Additive
    rotationSpace=BCS_BoneSpace
    alphaBoolBlend {blendOption=Linear}
    alphaScaleBiasClamp {interpSpeedDecreasing=1, interpSpeedIncreasing=1}
    boneToModify {boneName=dyn_fan_parent}
   }
   n2 = "Transform (Modify) Bone - Bone: dyn_fan_parent"(ComponentPose=@n1, Alpha=1, Translation="0,0,0", Rotation=Get Spinning_Rotator_Faster() -> (Spinning_Rotator_Faster:Rotator), Scale="1,1,1") {
    rotationMode=BMM_Additive
    rotationSpace=BCS_BoneSpace
    alphaBoolBlend {blendOption=Linear}
    alphaScaleBiasClamp {interpSpeedDecreasing=1, interpSpeedIncreasing=1}
    boneToModify {boneName=dyn_fan_parent}
   }
   n3 = "Transform (Modify) Bone - Bone: root"(ComponentPose=@n2, Translation=Get backpack_translation(), Rotation=Get backpack_rotate()) {
    rotationMode=BMM_Additive
    rotationSpace=BCS_BoneSpace
    translationMode=BMM_Additive
    translationSpace=BCS_BoneSpace
    alphaBoolBlend {blendOption=Linear}
    boneToModify {boneName=root}
   }
   n4 = RigidBody(ComponentPose=@n3, Alpha=Get Skydive_Alpha() -> (Skydive_Alpha:double)) {
    bFreezeIncomingPoseOnStart=true
    bOverrideWorldGravity=true
    overridePhysicsAsset=&/M_MED_Assemble_K_Pack_Physics
    simulationSpace=BaseBoneSpace
    alphaBoolBlend {blendOption=Linear}
    baseBoneRef {boneName=root}
    componentAppliedLinearAccClamp=<3000 3000 3000>
    componentLinearAccScale=<0.2 0.2 0.2>
    componentLinearVelScale=<0.2 0.2 0.2>
    overrideWorldGravity {z=-980}
   }
   Output Pose(Result=Component To Local(ComponentPose=@n4)) {layerGroup=None}
  ]}
 EventGraph {nodes [
   n1 = Event BlueprintUpdateAnimation()
   > n2 = Sequence!(execute) -> (then_0, then_1, then_2:Exec)
   n3 = Get bIsSkyDiving()
   n4 = Get bIsSkydiveDiveMode()
   n5 = Get bIsSkydiveFloating()
   > Set Skydive_Alpha(Skydive_Alpha:double=SelectFloat(A=0.5, B=1, bPickA=OR Boolean!(A=@n3, B=@n4, C:bool=@n5) -> (ReturnValue))) -> (Output_Get:double)
   n6 = GetOwnerSkeletalMeshComponent()
   n7 = BreakRotator(InRot=BreakTransform(InTransform=GetDeltaTransformFromRefPose(self=@n6, BoneName=spine_05, BaseName=spine_01)).Rotation)
   Set backpack_rotate(execute=@n2.then_1, backpack_rotate=MakeRotator(Roll=MapRangeClamped(Value=@n7.Yaw, InRangeA=0, InRangeB=-30, OutRangeB=20)))
   > Set backpack_translation(backpack_translation=MakeVector(Y=MapRangeClamped(Value=@n7.Yaw, InRangeA=0, InRangeB=-30, OutRangeB=-3)))
   n8 = Branch(execute=@n2.then_2, Condition="float >= float"(A=Get PawnSpeed2D(), B=200) {is=title K2Node_CallFunction})
   n9 = GetRealTimeSeconds()
   n10 = MakeRotator(Roll=float * float(A=float / float(A=@n9, B=-720) {is=title K2Node_CallFunction}, B=900000))
   n11 = GetRealTimeSeconds()
   n12 = MakeRotator(Roll=float * float(A=float / float(A=@n11, B=-720), B=400000))
   > Set Spinning_Rotator_Faster(Spinning_Rotator_Faster:Rotator=RInterpTo!(Current:Rotator=@n10, Target:Rotator=@n12, DeltaTime:float=@n1.DeltaTimeX, InterpSpeed:float=0.3) -> (ReturnValue:Rotator)) -> (Output_Get:Rotator)
   Set Spinning_Rotator(execute=@n8.else, Spinning_Rotator:Rotator=RInterpTo(Current=@n12, Target=@n10, DeltaTime=0, InterpSpeed=1)) -> (Output_Get:Rotator)
   TryGetPawnOwner()
  ]}
]
variables [
 backpack_rotate {type=Rotator}
 backpack_translation {type=Vector}
 Skydive_Alpha {type=number}
 Spinning_Rotator {type=Rotator}
 Spinning_Rotator_Faster {type=Rotator}
]
